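# Unlit translucent Niagara particle material for a soft radial gradient flare with an animated rainbow tint.
# A sphere mask shaped by dynamic-parameter power and particle alpha forms the fade, with quality-switched depth and near-camera fading plus optional view-angle and camera-depth fades.
# Distance from UV center scrolls a rainbow texture over time and builds a contrast mask blending it with particle color.
# Emissive is scaled by time of day.
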Material /WildEstateConsumables/Gameplay/SteelKite/FX/Effects/Fort_Effects/Materials/Flares/M_Radial_Gradient_trans_rainbow
graph {
 properties {
  bUsedWithNanite=true
  bUsedWithNiagaraMeshParticles=true
  bUsedWithNiagaraRibbons=true
  bUsedWithNiagaraSprites=true
  blendMode=BLEND_Translucent
  pixelDepthOffsetMode=PDOM_Legacy
  shadingModel=MSM_Unlit
  usedShadingModels=Unlit
 }
 nodes [
  n1 = SphereMask(
   A=TextureCoordinate()
   B=Constant() {r=0.5}
   Radius=ScalarParameter() {defaultValue=0.5, parameterName=Radius}
   Hardness=ScalarParameter() {parameterName=Hardness}
  ) {hardnessPercent=0}
  n2 = DynamicParameter() {defaultValue=<rgba 3 1 1 1>, paramNames [Power, DepthFadeDistMult, Param3, Param4]}
  n3 = Clamp(None=@n2.Power) {clampMode=CMODE_ClampMin, maxDefault=100, minDefault=1}
  n4 = ParticleColor()
  n5 = Constant() {r=1}
  n6 = DepthFade(
   Opacity=MaterialFunctionCall(Distance=ScalarParameter() {defaultValue=100, parameterName=NearCameraFadeDist}) {materialFunction=&Near_Camera_Fade}.Result
   FadeDistance=Multiply(A=@n2.DepthFadeDistMult, B=ScalarParameter() {defaultValue=100, parameterName=DepthFadeDist})
  )
  n7 = QualitySwitch(Default=@n5, Low=@n5, High=@n6, Medium=@n5, Epic=@n6)
  n8 = ScalarParameter() {defaultValue=8, parameterName=AngleFade}
  n9 = Multiply(
   A=Multiply(A=Power(Base=@n1, Exp=@n3) {constExponent=3}, B=Multiply(A=@n4.A, B=@n7))
   B=StaticSwitchParameter(
    True=MaterialFunctionCall("FadeSensitivity (optional)"=@n8) {materialFunction=&FadeBasedOnViewAngle}.Result
    False=Constant() {r=1}
   ) {parameterName=Use AngleFade}
  )
  n10 = ScalarParameter() {defaultValue=100, parameterName=CameraDepthFade_Length}
  n11 = ScalarParameter() {defaultValue=110, parameterName=CameraDepthFade_Offset}
  LinearInterpolate(
   A=@n9
   B=Multiply(
    A=@n9
    B=MaterialFunctionCall(Fade Length=@n10, Fade Offset=@n11) {materialFunction=&CameraDepthFade}.Result
   )
   Alpha=ScalarParameter() {parameterName=Switch_CameraDepthFade}
  )
  n12 = Distance(A=TextureCoordinate(), B=Constant() {r=0.5})
  n13 = ScalarParameter() {parameterName=Tile}
  n14 = DynamicParameter() {parameterIndex=1, defaultValue=<rgba 10 0.5 1 -0.75>, paramNames [Inner Glow Pow, Inner GLow Strength, Outline Str, Color Speed]}
  n15 = Frac(None=Multiply(A=Time(), B=@n14.Color Speed))
  n16 = TextureSample(UVs=Add(A=Multiply(A=@n12, B=@n13), B=OneMinus(None=@n15))) {texture=/WildEstateConsumables/Gameplay/SteelKite/FX/Textures/Rainbow_Colors_basecolor}
  n17 = VectorParameter() {parameterName=options, defaultValue=<rgba 1 1 0.15 3>}
  n18 = Multiply(A=Power(Base=@n12, Exp=@n17.R), B=@n17.G)
  n19 = Saturate(None=MaterialFunctionCall(In=Add(A=@n18, B=@n17.B), Contrast=@n17.A) {materialFunction=&CheapContrast}.Result)
  n20 = ScalarParameter() {defaultValue=1, parameterName="Value_Morning (S)"}
  n21 = ScalarParameter() {defaultValue=1, parameterName="Value_Day (S)"}
  n22 = ScalarParameter() {defaultValue=1, parameterName="Value_Evening (S)"}
  n23 = ScalarParameter() {defaultValue=1, parameterName="Value_Night (S)"}
  MaterialFunctionCall(Emissive=Multiply(
    A=LinearInterpolate(A=@n4.RGB, B=@n16.RGB, Alpha=@n19)
    B=MaterialFunctionCall(Value_Morning=@n20, Value_Day=@n21, Value_Evening=@n22, Value_Night=@n23) {materialFunction=&Scalar_Time_of_Day}.Result
   )) {materialFunction=&MF_FX_EmissiveAdjustments}
  ScalarParameter() {defaultValue=40, parameterName="GlowMeshStencilIndex (S)"}
  StaticBoolParameter() {defaultValue=true, parameterName=UseCustomDepth}
  StaticBoolParameter() {parameterName=UseCustomStencil}
 ]
}
